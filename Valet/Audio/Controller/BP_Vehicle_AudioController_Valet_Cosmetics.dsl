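; Audio controller Blueprint for Valet vehicles that plays a tire impact sound when an attached wheel's tire mod takes a health change.
; On init it casts the controller's vehicle actor to Valet_BasicCar_Vehicle and binds to its OnTireModHealthChanged dispatcher.
; The handler breaks the wheel struct and, if current health is above zero, plays the off-road tire impact cue at the wheel mesh's location.
; A comment marks this as Valet-only logic intended to move elsewhere later.
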(event VehicleAudio|EventOnInit
  (|CallParentFunction)
  (bind _asvalet_basic_car_vehicle (Utilities|Casting|CastToValet_BasicCar_Vehicle (AudioController|GetVehicleActor)))
  (EventDispatchers|AssignOnTireModHealthChanged _asvalet_basic_car_vehicle (EventDispatchers|CreateEvent :!event On%20Tire%20Mod%20Health%20Changed)))

(fn OnTireModHealthChanged
  (params (WheelData :type Attached%20Wheel%20Structure%20%28by%20ref%29))
  (bind (_skeletalmeshcomponent _currenthealth) (Utilities|Struct|BreakAttachedWheel WheelData))
  ;[1 1 1] Hopefully we can move this else where later, for now it works for only valet vehicles.
  (if (> _currenthealth 0)
    (Audio|PlaySoundatLocation "/ValetMods/Audio/OffRoadTire/ValetMod_OffRoadTire_Impact_Cue.ValetMod_OffRoadTire_Impact_Cue" (Transformation|GetWorldLocation _skeletalmeshcomponent))))

(fn ConstructionScript
  (|CallParentFunction))
